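# Control Rig for the spectral GateHound tail cosmetic, procedurally animating the snake head at the tip of the tail chain.
# Chain harmonics drive a noisy wave on the upper head chain, with pendulum off and debug drawing on.
# The lower jaw's rotation is the inverse of the upper head's local rotation times a small fixed offset, mirroring its motion.
# A final ModifyTransforms step applies fixed pose adjustments around the tail tip and head.
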ControlRigBlueprint /BRCosmetics/Accessories/FORT_Tails/Tail_M_MED_GateHound_Spectral/Meshes/M_MED_GateHound_Spectral_Tail_CtrlRig
graphs [
 RigVMFunctionLibrary {links [], nodes []}
 RigVMModel {nodes [
   RigUnit_BeginExecution = &RigUnit_BeginExecution()
   /Script/ControlRig.RigUnit_ChainHarmonicsPerItem(
    ChainRoot=(Type=Bone,Name="def_snake_head_upper_root")
    DrawWorldOffset=(Rotation=(X=0,Y=0,Z=0,W=1),Translation=(X=0,Y=0,Z=0),Scale3D=(X=1,Y=1,Z=1))
    Pendulum=(bEnabled=false,PendulumStiffness=2,PendulumGravity=(X=0,Y=0,Z=0),PendulumBlend=0.75,PendulumDrag=0.98,PendulumMinimum=0,PendulumMaximum=1,PendulumEase=Linear,UnwindAxis=(X=0,Y=1,Z=0),UnwindMinimum=0.2,UnwindMaximum=0.05)
    Reach=(bEnabled=true,ReachTarget=(X=0,Y=0,Z=0),ReachAxis=(X=1,Y=0,Z=0),ReachMinimum=0,ReachMaximum=0,ReachEase=Linear)
    Speed=(X=0,Y=4,Z=3)
    Wave=(bEnabled=True,WaveFrequency=(X=1,Y=0.5,Z=0.8),WaveAmplitude=(X=0,Y=0.35,Z=0),WaveOffset=(X=1,Y=1,Z=0),WaveNoise=(X=0,Y=0.3,Z=3),WaveMinimum=0.2,WaveMaximum=1,WaveEase=SineEaseInOut)
    WaveCurve=(EditorCurveData=(Keys=((InterpMode=RCIM_Cubic),(InterpMode=RCIM_Cubic,Time=1,Value=1))))
    bDrawDebug=True
    ExecutePin=@RigUnit_BeginExecution.ExecutePin
   ) {is=struct RigVMUnitNode}
   RigUnit_GetTransform = &RigUnit_GetTransform(Item=(Type=Bone,Name="def_snake_head_upper"), Space=LocalSpace, bInitial=False)
   Inverse = /Script/RigVM.RigVMFunction_MathQuaternionInverse(Value=@RigUnit_GetTransform.Transform.Rotation) {is=struct RigVMUnitNode, template="Inverse::Execute(in Value,out Result)"}
   Multiply = /Script/RigVM.RigVMFunction_MathQuaternionMul(B=(X=-0.0086,Y=-0.0175,Z=0.0086,W=0.9998), A=@Inverse.Result) {is=struct RigVMUnitNode, template="Multiply::Execute(in A,in B,out Result)"}
   > RigUnit_SetRotation = &RigUnit_SetRotation(Item=(Type=Bone,Name="def_snake_head_lower"), Weight=1, bInitial=False, Value=@Multiply.Result)
   > &RigUnit_ModifyTransforms(
    ItemToModify="((Item=(Type=Bone,Name=\"C_Tail_A_8_Jnt\"),Transform=(Rotation=(X=0,Y=0.0188,Z=0,W=0.9998),Translation=(X=0,Y=0,Z=0),Scale3D=(X=1,Y=1,Z=1))),(Item=(Type=Bone,Name=\"def_snake_head_upper_root\"),Transform=(Rotation=(X=0,Y=0.0732,Z=0,W=0.9973),Translation=(X=0,Y ...(truncated)"
    Weight=1
    WeightMinimum=0
   )
  ]}
]
hierarchy [
 C_Root_Main_Root_Jnt {loc=<0 0 0>, rot=<0 0 0 1>}
 C_Tail_A_Base_Jnt {parent=C_Root_Main_Root_Jnt, loc=<1.4997 -0.3348 0.4223>, rot=<-0.5 0.5 0.5 -0.5>}
 C_Tail_A_1_Jnt {parent=C_Tail_A_Base_Jnt, loc=<8.8758 -0.0429 0.1469>, rot=<0.7071 0.0063 0.0115 -0.707>}
 C_Tail_A_2_Jnt {parent=C_Tail_A_1_Jnt, loc=<8.8775 0.0376 0.011>, rot=<0 0.0009 -0.0031 -1>}
 C_Tail_A_3_Jnt {parent=C_Tail_A_2_Jnt, loc=<8.879 0.0015 0.0004>, rot=<0 0 -0.0001 -1>}
 C_Tail_A_4_Jnt {parent=C_Tail_A_3_Jnt, loc=<8.8778 0 0>, rot=<0 0 0 1>}
 C_Tail_A_5_Jnt {parent=C_Tail_A_4_Jnt, loc=<8.8778 0 0>, rot=<0 0 0 1>}
 C_Tail_A_6_Jnt {parent=C_Tail_A_5_Jnt, loc=<8.8775 0 0>, rot=<0 0 0 1>}
 C_Tail_A_7_Jnt {parent=C_Tail_A_6_Jnt, loc=<8.8776 0 0>, rot=<0 0 0 1>}
 C_Tail_A_8_Jnt {parent=C_Tail_A_7_Jnt, loc=<8.8788 0 0>, rot=<0 0 0 1>}
 def_snake_head_upper_root {parent=C_Tail_A_8_Jnt, loc=<6.7156 0 0.6449>, rot=<0 0.2405 0 -0.9707>}
 def_snake_head_upper {parent=def_snake_head_upper_root, loc=<0.2717 0 0>, rot=<0 0 0 1>}
 def_snake_head_lower_root {parent=C_Tail_A_8_Jnt, loc=<6.7269 0 -1.2066>, rot=<0 -0.1416 0 -0.9899>}
 def_snake_head_lower {parent=def_snake_head_lower_root, loc=<0.3939 0 0>, rot=<0 0 0 1>}
 Root {parent=C_Root_Main_Root_Jnt, type=Socket, loc=<0 0 0>, rot=<0 0 0 1>}
]
